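# Material function for the MincePounce character that blends incoming material attributes toward an emissive hologram look with its own opacity, with an optional channel-selectable HoloMask.
# Scrolling scanlines over UV0, UV1 or scaled local position, plus optional dual sine waves, modulate glow and opacity.
# Fresnel and height-attenuated inner/outer colors, optionally tinted by base emissive or a gradient map, form the glow.
# An optional object-space inner-light point adds directional variation.
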MaterialFunction /BRCosmetics/Characters/Player/Female/Medium/Bodies/F_MED_MincePounce/Materials/MF_MincePounceHologram
graph {nodes [
  n1 = VectorParameter() {group=Hologram - InnerLight, parameterName=InnerLight_ObjectPosition, sortPriority=1, defaultValue=<rgba 0 0 250 0>}
  n2 = LocalPosition() {localOrigin=InstancePreSkinning}
  n3 = Normalize(VectorInput=Subtract(A=@n1.RGB, B=VertexInterpolator(VS=@n2.XYZ).PS))
  DotProduct(A=PixelNormalWS(), B=@n3)
  n4 = NamedRerouteUsage()
  n5 = ScalarParameter() {
   defaultValue=0.9
   desc="Setting this too low will produce \"anotomically correct\" highlights - if you get what I'm saying"
   group=Hologram - General
   parameterName=RoughnessAmount
   sortPriority=5
  }
  n6 = SetMaterialAttributes!(
   MaterialAttributes=@n4
   BaseColor=Constant()
   Roughness=@n5
   Specular=Constant()
   Metallic=Constant()
   EmissiveColor=NamedRerouteUsage()
   Opacity=NamedRerouteUsage()
   Normal=Constant3Vector() {constant=<rgba 0 0 1 0>}
  ) {attributeSetTypes [
    #1
    #2
    #3
    #4
    #5
    #6
    #7
   ]}
  FunctionOutput(None=BlendMaterialAttributes(A=@n4, B=@n6, Alpha=NamedRerouteUsage()))
  n7 = NamedRerouteUsage()
  n8 = ScalarParameter() {defaultValue=0.8, group=Hologram - General, parameterName=OpacityBiasScale, sortPriority=10}
  Multiply(
   A=Add(A=Multiply(A=@n7, B=@n8), B=OneMinus(None=@n8))
   B=MaterialFunctionCall(Distance=ScalarParameter() {defaultValue=200, group=Hologram - General, parameterName=NearCamera_FadeDistance, sortPriority=11}) {materialFunction=&Near_Camera_Fade}.Result
  )
  NamedRerouteDeclaration(None=StaticBoolParameter() {group=Hologram - General, parameterName=UseUV2, sortPriority=2}) {name=UseUV2, nodeColor=<rgba 0.888 0.3416 1 1>}
  n9 = NamedRerouteUsage()
  n10 = Panner(
   Coordinate=@n9
   Speed=AppendVector(A=Constant(), B=ScalarParameter() {defaultValue=1, group=Hologram - Scanlines, parameterName=ScanlineSpeed, sortPriority=0})
  ) {speedY=0.03}
  n11 = ScalarParameter() {defaultValue=150, group=Hologram - Scanlines, parameterName=ScanlineAmount, sortPriority=1}
  n12 = Frac(None=StaticComponentMaskParameter(None=Multiply(A=@n10, B=@n11) {constB=200}) {defaultG=true, group=Hologram - Scanlines, parameterName="Scanline Mask "})
  n13 = Multiply(A=@n12, B=OneMinus(None=@n12))
  n14 = ScalarParameter() {defaultValue=-0.15, group=Hologram - Scanlines, parameterName=ScanlineSize, sortPriority=2}
  n15 = ScalarParameter() {defaultValue=0.1, group=Hologram - Scanlines, parameterName=Scanline_SmoothCeilRamp}
  n16 = LinearInterpolate(
   A=ScalarParameter() {defaultValue=0.5, group=Hologram - Scanlines, parameterName=ScanlinesDark, sortPriority=3}
   B=ScalarParameter() {defaultValue=1, group=Hologram - Scanlines, parameterName=ScanlinesLight, sortPriority=4}
   Alpha=MaterialFunctionCall!(In=Add(A=@n13, B=@n14), Ramp Width=@n15) {materialFunction=/Engine/Functions/Engine_MaterialFunctions01/Math/SmoothCeil}.Result
  )
  n17 = Panner(
   Coordinate=@n9
   Speed=AppendVector(A=Constant(), B=ScalarParameter() {defaultValue=10, group=Hologram - Scanlines, parameterName=Secondary_ScanlineSpeed, sortPriority=6})
  ) {speedY=0.03}
  n18 = ScalarParameter() {defaultValue=10, group=Hologram - Scanlines, parameterName=Secondary_ScanlineAmount, sortPriority=7}
  n19 = Frac(None=StaticComponentMaskParameter(None=Multiply(A=@n17, B=@n18) {constB=200}) {defaultG=true, group=Hologram - Scanlines, parameterName="Scanline Mask "})
  n20 = Multiply(A=@n19, B=OneMinus(None=@n19))
  n21 = ScalarParameter() {defaultValue=-0.15, group=Hologram - Scanlines, parameterName=Secondary_ScanlineSize, sortPriority=8}
  n22 = LinearInterpolate(
   A=ScalarParameter() {defaultValue=0.5, group=Hologram - Scanlines, parameterName=Secondary_ScanlinesDark, sortPriority=9}
   B=ScalarParameter() {defaultValue=1, group=Hologram - Scanlines, parameterName=Secondary_ScanlinesLight, sortPriority=10}
   Alpha=MaterialFunctionCall(In=Add(A=@n20, B=@n21), Ramp Width=@n15) {materialFunction=&SmoothCeil}.Result
  )
  NamedRerouteDeclaration(None=StaticSwitchParameter(True=Subtract(A=@n16, B=@n22), False=@n16) {group=Hologram - Scanlines, parameterName=SubtractSecondaryScanlines, sortPriority=5}) {name=ScanLines, nodeColor=<rgba 1 0.7734 0 1>}
  n23 = FunctionInput() {bUsePreviewValueAsDefault=true, inputName=MA, inputType=FunctionInput_MaterialAttributes}
  NamedRerouteDeclaration(None=GetMaterialAttributes(None=NamedRerouteDeclaration(None=@n23) {name=InMaterial, nodeColor=<rgba 0 1 1 1>}) {attributeGetTypes [#5]}.EmissiveColor) {name=InEmissive, nodeColor=<rgba 1 0 0.8647 1>}
  n24 = NamedRerouteUsage()
  n25 = ScalarParameter() {group=Hologram - General, parameterName=HoloOpacity}
  n26 = Multiply(A=@n24, B=Multiply(A=@n25, B=NamedRerouteUsage()))
  n27 = NamedRerouteUsage()
  n28 = ScalarParameter() {defaultValue=1, group=Hologram - Waves, parameterName=WaveOpacity, sortPriority=5}
  n29 = Add(
   A=Add(A=@n26, B=Multiply(A=@n27, B=@n28))
   B=Multiply(A=NamedRerouteUsage(), B=ScalarParameter() {defaultValue=1, group=Hologram - Waves, parameterName=Secondary_WaveOpacity, sortPriority=10})
  )
  n30 = NamedRerouteUsage()
  NamedRerouteDeclaration(None=StaticSwitchParameter(True=StaticSwitch(True=@n29, False=@n26, Value=@n30), False=Multiply(A=@n24, B=@n25)) {defaultValue=true, parameterName=UseScanlines}) {name=OutOpacity, nodeColor=<rgba 0.5234 0 1 1>}
  n31 = ScalarParameter() {defaultValue=0.2, group=Hologram - Waves, parameterName=WaveScale, sortPriority=2}
  n32 = Multiply(
   A=NamedRerouteUsage()
   B=Multiply(A=ScalarParameter() {defaultValue=1.5, group=Hologram - Waves, parameterName=WaveSpeed, sortPriority=1}, B=@n31)
  ) {constB=0.06}
  n33 = Multiply(A=NamedRerouteUsage()) {constB=-0.01}
  n34 = Multiply(A=@n31, B=@n33) {constB=0.5}
  n35 = Sine(None=Subtract(A=Add(A=@n32, B=@n34)) {constB=0.075})
  NamedRerouteDeclaration(None=SmoothStep(Min=ScalarParameter() {defaultValue=0.8, group=Hologram - Waves, parameterName=WaveMin, sortPriority=3}, Value=Saturate(None=@n35)) {constMin=0.5}) {name=PrimaryWave, nodeColor=<rgba 0 1 0.2031 1>}
  n36 = ScalarParameter() {defaultValue=1.5, group=Hologram - Waves, parameterName=Secondary_WaveScale, sortPriority=6}
  n37 = Multiply(A=@n33, B=@n36)
  n38 = Multiply(
   A=NamedRerouteUsage()
   B=Multiply(A=ScalarParameter() {defaultValue=0.5, group=Hologram - Waves, parameterName=Secondary_WaveSpeed, sortPriority=7}, B=@n36)
  ) {constB=0.06}
  n39 = Sine(None=Subtract(A=Add(A=@n37, B=@n38)) {constB=0.075})
  NamedRerouteDeclaration(None=SmoothStep(Min=ScalarParameter() {group=Hologram - Waves, parameterName=Secondary_WaveMin, sortPriority=8}, Value=Saturate(None=@n39)) {constMin=0.5}) {name=SecondaryWave, nodeColor=<rgba 0.125 0 1 1>}
  NamedRerouteDeclaration(None=MaterialFunctionCall(
    Brightness=ScalarParameter() {defaultValue=1, group=Hologram - General, parameterName=FresnelBrightness, sliderMin=4, sortPriority=6}
    Exponent=ScalarParameter() {defaultValue=1.5, group=Hologram - General, parameterName=FresnelExponent, sortPriority=7}
    Vertex / Pixel Normal=PixelNormalWS()
   ) {materialFunction=&MF_Fresnel_FX}.Fresnel Inverted) {name=Fresnel, nodeColor=<rgba 1 0 0.2812 1>}
  NamedRerouteDeclaration(None=StaticBoolParameter() {group=Hologram - General, parameterName=UseLocalPosHologram, sortPriority=2}) {name=UseLocalPosHologram, nodeColor=<rgba 0.888 0.3416 1 1>}
  n40 = ScalarParameter() {defaultValue=0.2, group=Hologram - InnerLight, parameterName=InnerOpacityDark, sortPriority=4}
  n41 = ScalarParameter() {defaultValue=1, group=Hologram - InnerLight, parameterName=InnerOpacityBright, sortPriority=5}
  n42 = ConstantBiasScale(None=@n3)
  n43 = Constant() {r=1}
  NamedRerouteDeclaration(None=StaticSwitch(True=LinearInterpolate(A=@n40, B=@n41, Alpha=@n42), False=@n43, Value=NamedRerouteUsage())) {name=InnerOpacity, nodeColor=<rgba 0 0.6016 1 1>}
  n44 = ScalarParameter() {defaultValue=0.05, group=Hologram - InnerLight, parameterName=InnerLightingDark, sortPriority=2}
  n45 = ScalarParameter() {defaultValue=3, group=Hologram - InnerLight, parameterName=InnerLightingBright, sortPriority=3}
  NamedRerouteDeclaration(None=StaticSwitch(True=LinearInterpolate(A=@n44, B=@n45, Alpha=@n42), False=@n43, Value=NamedRerouteUsage())) {name=InnerLighting, nodeColor=<rgba 0.9219 1 0 1>}
  NamedRerouteDeclaration(None=Time()) {name=Time, nodeColor=<rgba 1 0.3281 0 1>}
  NamedRerouteDeclaration(None=MaterialFunctionCall(
    Brightness=ScalarParameter() {defaultValue=1, group=Hologram - General, parameterName=FresnelCenterBrightness, sliderMin=4, sortPriority=6}
    Exponent=ScalarParameter() {defaultValue=1.5, group=Hologram - General, parameterName=FresnelCenterExponent, sortPriority=7}
    Vertex / Pixel Normal=PixelNormalWS()
   ) {materialFunction=&MF_Fresnel_FX}.Fresnel Inverted) {name=FresnelCenter, nodeColor=<rgba 1 0 0.2812 1>}
  n46 = LocalPosition() {localOrigin=InstancePreSkinning}
  n47 = NamedRerouteDeclaration(None=ChannelMaskParameter(None=VertexInterpolator(VS=@n46.XYZ).PS) {group=Hologram - General, parameterName=FlowDirection, sortPriority=3}.R) {name=LocalPos, nodeColor=<rgba 1 0.75 0 1>}
  n48 = ScalarParameter() {defaultValue=0.01, group=Hologram - General, parameterName=HologramConstantScale, sortPriority=4}
  n49 = StaticSwitch(True=TextureCoordinate() {coordinateIndex=1}, False=TextureCoordinate(), Value=NamedRerouteUsage())
  NamedRerouteDeclaration(None=StaticSwitch(
    True=Multiply(A=@n47, B=@n48) {constB=0.01}
    False=NamedRerouteDeclaration(None=@n49) {name=UVs, nodeColor=<rgba 0.4297 0 1 1>}
    Value=NamedRerouteUsage()
   )) {name=HologramUVs, nodeColor=<rgba 0 1 0.8125 1>}
  NamedRerouteDeclaration(None=StaticBoolParameter() {group=Hologram - Waves, parameterName=UseWaves, sortPriority=0}) {name=UseWaves, nodeColor=<rgba 0 0.4141 1 1>}
  n50 = Multiply(
   A=Multiply(A=ScalarParameter() {defaultValue=1, parameterName=GradientScale}) {constB=0.1}
   B=Add(A=ScalarParameter() {parameterName=GradientPos}, B=NamedRerouteUsage())
  ) {constB=0.01}
  NamedRerouteDeclaration(None=LinearInterpolate(
    A=VectorParameter() {parameterName=Gradient_Color_1, defaultValue=<rgba 0 0 0 1>}.RGB
    B=VectorParameter() {parameterName=Gradient_Color_2, defaultValue=<rgba 0 0 0 1>}.RGB
    Alpha=TextureSampleParameter2D(UVs=@n50) {parameterName=GradientMap, samplerType=SAMPLERTYPE_Masks, texture=&T_WhiteMask}.RGB
   )) {name=Gradient, nodeColor=<rgba 0 0.9297 1 1>}
  NamedRerouteDeclaration(None=StaticBoolParameter() {group=Hologram - InnerLight, parameterName=UseInnerLight, sortPriority=0}) {name=UseInnerLight, nodeColor=<rgba 0 0.0337 1 1>}
  n51 = FunctionInput() {bUsePreviewValueAsDefault=true, inputName=HoloMask, sortPriority=2}
  NamedRerouteDeclaration(None=StaticSwitchParameter(True=ChannelMaskParameter(None=@n51) {group=Hologram - General, parameterName=HoloMaskChannel, sortPriority=1}.R, False=Constant() {r=1}) {defaultValue=true, group=Hologram - General, parameterName=UseHoloMask, sortPriority=0}) {name=HoloMask, nodeColor=<rgba 0 0.6953 1 1>}
  n52 = VectorParameter() {group=Hologram - Color, parameterName=GlowColor_Outer, sortPriority=1, defaultValue=<rgba 3.1771 3.7137 5 1>}
  n53 = StaticSwitchParameter(
   True=NamedRerouteUsage()
   False=StaticSwitchParameter(True=NamedRerouteUsage(), False=Constant() {r=1}) {parameterName=UseGradeintBackground}
  ) {defaultValue=true, group=Hologram - Color, parameterName=UseBaseEmissiveForColor, sortPriority=2}
  n54 = ComponentMask(None=WorldPosition().XYZ) {b=true}
  n55 = ScalarParameter() {group=Hologram - General, parameterName=FresnelEdgesZ_AttenOffset, sortPriority=8}
  n56 = ComponentMask(None=MaterialFunctionCall() {materialFunction=&ObjectLocalBounds}.Local Bounds Size) {b=true}
  n57 = ScalarParameter() {group=Hologram - General, parameterName=FresnelEdgesZ_AttenScale, sortPriority=9}
  n58 = Saturate(None=Divide(A=Add(A=@n54, B=@n55) {constB=10}, B=Multiply(A=@n56, B=@n57) {constB=1.2}) {constB=384})
  n59 = NamedRerouteUsage()
  n60 = VectorParameter() {group=Hologram - Color, parameterName=GlowColor_Inner, sortPriority=0, defaultValue=<rgba 0.4199 0.4465 0.7396 1>}
  n61 = LinearInterpolate(A=Multiply(A=@n60.RGB, B=@n53), B=Multiply(A=@n52.RGB, B=@n53), Alpha=Saturate(None=Multiply(A=@n58, B=@n59) {constA=1}))
  n62 = NamedRerouteUsage()
  n63 = Add(
   A=@n62
   B=Multiply(A=ScalarParameter() {defaultValue=0.5, group=Hologram - Waves, parameterName=WaveBrightness, sortPriority=4}, B=NamedRerouteUsage())
  )
  n64 = Multiply(
   A=ScalarParameter() {defaultValue=0.5, group=Hologram - Waves, parameterName=Secondary_WaveBrightness, sortPriority=9}
   B=NamedRerouteUsage()
  )
  n65 = Multiply(A=@n61, B=StaticSwitch(True=Add(A=@n63, B=@n64), False=@n62, Value=NamedRerouteUsage()))
  NamedRerouteDeclaration(None=Multiply(A=NamedRerouteUsage(), B=StaticSwitchParameter(True=@n65, False=@n61) {defaultValue=true, parameterName=UseScanlines})) {name=OutEmissive, nodeColor=<rgba 1 0.2109 0 1>}
  WorldPosition()
 ]}
