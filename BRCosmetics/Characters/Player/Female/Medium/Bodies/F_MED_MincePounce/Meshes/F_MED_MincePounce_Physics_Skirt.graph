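# Physics asset for the skirt of the F_MED_MincePounce female player body, presumably driving its secondary motion.
# Nine capsule chains (A–D per side, plus back) hang from MSTR_skirt anchor bones through tight swing and twist limits, alongside pelvis, thigh, calf and foot capsules.
# Angularly free cross-constraints tie neighboring chains into a connected sheet.
# The back chain links only to left D, leaving a gap at right D.
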PhysicsAsset /BRCosmetics/Characters/Player/Female/Medium/Bodies/F_MED_MincePounce/Meshes/F_MED_MincePounce_Physics_Skirt
bodies {
 MSTR_skirt_A_l [{shapeType=Capsule, length=1, radius=1, shapeName=MSTR_skirt_A_l_capsule}]
 MSTR_skirt_A_r [{shapeType=Capsule, length=1, radius=1, shapeName=MSTR_skirt_A_r_capsule}]
 MSTR_skirt_B_l [{shapeType=Capsule, length=1, radius=1, shapeName=MSTR_skirt_B_l_capsule}]
 MSTR_skirt_B_r [{shapeType=Capsule, length=1, radius=1, shapeName=MSTR_skirt_B_r_capsule}]
 MSTR_skirt_C_l [{shapeType=Capsule, length=1, radius=1, shapeName=MSTR_skirt_C_l_capsule}]
 MSTR_skirt_C_r [{shapeType=Capsule, length=1, radius=1, shapeName=MSTR_skirt_C_r_capsule}]
 MSTR_skirt_D_l [{shapeType=Capsule, length=1, radius=1, shapeName=MSTR_skirt_D_l_capsule}]
 MSTR_skirt_D_r [{shapeType=Capsule, length=1, radius=1, shapeName=MSTR_skirt_D_r_capsule}]
 MSTR_skirt_back [{shapeType=Capsule, length=1, radius=1, shapeName=MSTR_skirt_back_capsule}]
 calf_l [{shapeType=Capsule, length=35.4531, radius=7.551, shapeName=calf_l_capsule, center=<-26.6989 -2.663 -0.4505>, rotation=<rot 90 0 0>}]
 calf_r [{shapeType=Capsule, length=36.2559, radius=6.7482, shapeName=calf_r_capsule, center=<29.6012 2.025 0.8178>, rotation=<rot 90 0 0>}]
 foot_l [{shapeType=Capsule, length=12.9265, radius=8.2628, shapeName=foot_l_capsule, center=<-4.7366 7.4123 0>, rotation=<rot 0 2.5544 90>}]
 foot_r [{shapeType=Capsule, length=15.0974, radius=6.3389, shapeName=foot_r_capsule, center=<3.761 -6.9597 0>, rotation=<rot 0 4.6918 90>}]
 pelvis [{shapeType=Capsule, length=12.944, radius=8.5505, shapeName=pelvis_capsule, center=<-5.4021 -2.5347 0>, rotation=<rot 0 -90 0>}]
 skirt_A_l_1 [{shapeType=Capsule, length=0.2957, radius=1, shapeName=skirt_A_l_1_capsule, rotation=<rot 90 0 0>}]
 skirt_A_l_2 [{shapeType=Capsule, length=0.2957, radius=1, shapeName=skirt_A_l_2_capsule, rotation=<rot 90 0 0>}]
 skirt_A_l_3 [{shapeType=Capsule, length=0.2957, radius=1, shapeName=skirt_A_l_3_capsule, rotation=<rot 90 0 0>}]
 skirt_A_r_1 [{shapeType=Capsule, length=0.2957, radius=1, shapeName=skirt_A_r_1_capsule, center=<3.6421 0 2.8252>, rotation=<rot 90 0 0>}]
 skirt_A_r_2 [{shapeType=Capsule, length=0.2957, radius=1, shapeName=skirt_A_r_2_capsule, center=<3.6421 0 2.8252>, rotation=<rot 90 0 0>}]
 skirt_A_r_3 [{shapeType=Capsule, length=0.2957, radius=1, shapeName=skirt_A_r_3_capsule, center=<3.6421 0 2.8252>, rotation=<rot 90 0 0>}]
 skirt_B_l_1 [{shapeType=Capsule, length=1.8507, radius=3.4983, shapeName=skirt_B_l_1_capsule, center=<3.6421 0 3.6329>, rotation=<rot 90 0 0>}]
 skirt_B_l_2 [{shapeType=Capsule, length=2.5859, radius=3.863, shapeName=skirt_B_l_2_capsule, center=<3.6421 0 2.8362>, rotation=<rot 90 0 0>}]
 skirt_B_l_3 [{shapeType=Capsule, length=2.2371, radius=5.4118, shapeName=skirt_B_l_3_capsule, center=<3.6421 0 2.211>, rotation=<rot 90 0 0>}]
 skirt_B_r_1 [{shapeType=Capsule, length=1.1984, radius=4.1506, shapeName=skirt_B_r_1_capsule, center=<3.6421 0 2.8252>, rotation=<rot 90 0 0>}]
 skirt_B_r_2 [{shapeType=Capsule, length=0.7509, radius=5.1719, shapeName=skirt_B_r_2_capsule, center=<3.6421 0 3.8843>, rotation=<rot 90 0 0>}]
 skirt_B_r_3 [{shapeType=Capsule, length=1.3621, radius=6.8106, shapeName=skirt_B_r_3_capsule, center=<3.6421 0 4.5105>, rotation=<rot 90 0 0>}]
 skirt_C_l_1 [{shapeType=Capsule, length=0.6438, radius=4.7052, shapeName=skirt_C_l_1_capsule, center=<3.6421 0 3.1978>, rotation=<rot 90 0 0>}]
 skirt_C_l_2 [{shapeType=Capsule, length=1.4401, radius=5.0533, shapeName=skirt_C_l_2_capsule, center=<3.9731 0 3.0296>, rotation=<rot 90 0 0>}]
 skirt_C_l_3 [{shapeType=Capsule, length=0.4741, radius=6.0623, shapeName=skirt_C_l_3_capsule, center=<3.6421 0 3.8289>, rotation=<rot 90 0 0>}]
 skirt_C_l_4 [{shapeType=Capsule, length=0.5903, radius=8.4165, shapeName=skirt_C_l_4_capsule, center=<4.4055 0 6.3285>, rotation=<rot 90 0 0>}]
 skirt_C_r_1 [{shapeType=Capsule, length=0.2957, radius=5.0533, shapeName=skirt_C_r_1_capsule, center=<3.6421 0 2.8252>, rotation=<rot 90 0 0>}]
 skirt_C_r_2 [{shapeType=Capsule, length=0.7231, radius=5.0533, shapeName=skirt_C_r_2_capsule, center=<3.6421 0 2.8252>, rotation=<rot 90 0 0>}]
 skirt_C_r_3 [{shapeType=Capsule, length=1.1424, radius=5.5852, shapeName=skirt_C_r_3_capsule, center=<3.6421 0 2.8252>, rotation=<rot 90 0 0>}]
 skirt_C_r_4 [{shapeType=Capsule, length=2.6643, radius=8.3402, shapeName=skirt_C_r_4_capsule, center=<4.4978 0 4.9074>, rotation=<rot 90 0 0>}]
 skirt_D_l_1 [{shapeType=Capsule, length=0.3146, radius=4.8484, shapeName=skirt_D_l_1_capsule, center=<3.9718 0 3.5131>, rotation=<rot 90 0 0>}]
 skirt_D_l_2 [{shapeType=Capsule, length=3.7124, radius=7.2136, shapeName=skirt_D_l_2_capsule, center=<5.7197 0 4.258>, rotation=<rot 90 0 0>}]
 skirt_D_r_1 [{shapeType=Capsule, length=0, radius=5.2794, shapeName=skirt_D_r_1_capsule, center=<3.9718 -2.5124 3.4119>, rotation=<rot 90 0 0>}]
 skirt_D_r_2 [{shapeType=Capsule, length=4.7217, radius=6.0874, shapeName=skirt_D_r_2_capsule, center=<5.7197 -1.9516 3.0383>, rotation=<rot 90 0 0>}]
 skirt_back_1 [{shapeType=Capsule, length=1.7745, radius=3.5745, shapeName=skirt_back_1_capsule, center=<3.6421 0 2.8252>, rotation=<rot 90 0 0>}]
 skirt_back_2 [{shapeType=Capsule, length=3.808, radius=4.3451, shapeName=skirt_back_2_capsule, center=<3.6421 -1.5455 2.8252>, rotation=<rot 90 0 0>}]
 thigh_l [{
   shapeType=Capsule
   length=24.1595
   radius=9.8348
   shapeName=thigh_l_capsule
   center=<-14.407 -1.7484 2.5959>
   rotation=<rot 86.7638 179.9998 -172.8595>
  }]
 thigh_r [{
   shapeType=Capsule
   length=24.5366
   radius=10.0744
   shapeName=thigh_r_capsule
   center=<16.6606 1.9723 -2.5596>
   rotation=<rot 88.2265 179.9996 -178.3335>
  }]
}
constraints [
 {bone1Name=skirt_A_l_3, bone2Name=skirt_A_l_2, swing1LimitDegrees=30, swing2LimitDegrees=30, twistLimitDegrees=5}
 {bone1Name=skirt_A_l_2, bone2Name=skirt_A_l_1, swing1LimitDegrees=30, swing2LimitDegrees=30, twistLimitDegrees=5}
 {bone1Name=skirt_A_l_1, bone2Name=MSTR_skirt_A_l, swing1LimitDegrees=30, swing2LimitDegrees=30, twistLimitDegrees=5}
 {bone1Name=skirt_A_r_3, bone2Name=skirt_A_r_2, swing1LimitDegrees=30, swing2LimitDegrees=30, twistLimitDegrees=5}
 {bone1Name=skirt_A_r_2, bone2Name=skirt_A_r_1, swing1LimitDegrees=30, swing2LimitDegrees=30, twistLimitDegrees=5}
 {bone1Name=skirt_A_r_1, bone2Name=MSTR_skirt_A_r, swing1LimitDegrees=30, swing2LimitDegrees=30, twistLimitDegrees=5}
 {bone1Name=skirt_B_l_3, bone2Name=skirt_B_l_2, swing1LimitDegrees=30, swing2LimitDegrees=30, twistLimitDegrees=5}
 {bone1Name=skirt_B_l_2, bone2Name=skirt_B_l_1, swing1LimitDegrees=30, swing2LimitDegrees=30, twistLimitDegrees=5}
 {bone1Name=skirt_B_l_1, bone2Name=MSTR_skirt_B_l, swing1LimitDegrees=30, swing2LimitDegrees=30, twistLimitDegrees=5}
 {bone1Name=skirt_B_r_3, bone2Name=skirt_B_r_2, swing1LimitDegrees=30, swing2LimitDegrees=30, twistLimitDegrees=5}
 {bone1Name=skirt_B_r_2, bone2Name=skirt_B_r_1, swing1LimitDegrees=30, swing2LimitDegrees=30, twistLimitDegrees=5}
 {bone1Name=skirt_B_r_1, bone2Name=MSTR_skirt_B_r, swing1LimitDegrees=30, swing2LimitDegrees=30, twistLimitDegrees=5}
 {bone1Name=skirt_C_l_4, bone2Name=skirt_C_l_3, swing1LimitDegrees=30, swing2LimitDegrees=30, twistLimitDegrees=5}
 {bone1Name=skirt_C_l_3, bone2Name=skirt_C_l_2, swing1LimitDegrees=30, swing2LimitDegrees=20, twistLimitDegrees=5}
 {bone1Name=skirt_C_l_2, bone2Name=skirt_C_l_1, swing1LimitDegrees=30, swing2LimitDegrees=30, twistLimitDegrees=5}
 {bone1Name=skirt_C_l_1, bone2Name=MSTR_skirt_C_l, swing1LimitDegrees=30, swing2LimitDegrees=30, twistLimitDegrees=5}
 {bone1Name=skirt_C_r_4, bone2Name=skirt_C_r_3, swing1LimitDegrees=30, swing2LimitDegrees=30, twistLimitDegrees=5}
 {bone1Name=skirt_C_r_3, bone2Name=skirt_C_r_2, swing1LimitDegrees=30, swing2LimitDegrees=20, twistLimitDegrees=5}
 {bone1Name=skirt_C_r_2, bone2Name=skirt_C_r_1, swing1LimitDegrees=30, swing2LimitDegrees=30, twistLimitDegrees=5}
 {bone1Name=skirt_C_r_1, bone2Name=MSTR_skirt_C_r, swing1LimitDegrees=30, swing2LimitDegrees=30, twistLimitDegrees=5}
 {bone1Name=skirt_D_l_2, bone2Name=skirt_D_l_1, swing1LimitDegrees=7, swing2LimitDegrees=30, twistLimitDegrees=5}
 {bone1Name=skirt_D_l_1, bone2Name=MSTR_skirt_D_l, swing1LimitDegrees=7, swing2LimitDegrees=30, twistLimitDegrees=5}
 {bone1Name=skirt_D_r_2, bone2Name=skirt_D_r_1, swing1LimitDegrees=7, swing2LimitDegrees=30, twistLimitDegrees=5}
 {bone1Name=skirt_D_r_1, bone2Name=MSTR_skirt_D_r, swing1LimitDegrees=10, swing2LimitDegrees=30, twistLimitDegrees=5}
 {bone1Name=skirt_back_2, bone2Name=skirt_back_1, swing1LimitDegrees=7, swing2LimitDegrees=20, twistLimitDegrees=2}
 {bone1Name=skirt_back_1, bone2Name=MSTR_skirt_back, swing1LimitDegrees=7, swing2LimitDegrees=40, twistLimitDegrees=2}
 {
  bone1Name=skirt_B_r_3
  bone2Name=skirt_A_r_3
  swing1LimitDegrees=45
  swing1Motion=Free
  swing2LimitDegrees=45
  swing2Motion=Free
  twistLimitDegrees=45
  twistMotion=Free
 }
 {
  bone1Name=skirt_C_r_3
  bone2Name=skirt_B_r_3
  swing1LimitDegrees=45
  swing1Motion=Free
  swing2LimitDegrees=45
  swing2Motion=Free
  twistLimitDegrees=45
  twistMotion=Free
 }
 {
  bone1Name=skirt_D_r_2
  bone2Name=skirt_C_r_2
  swing1LimitDegrees=45
  swing1Motion=Free
  swing2LimitDegrees=45
  swing2Motion=Free
  twistLimitDegrees=45
  twistMotion=Free
 }
 {
  bone1Name=skirt_B_l_3
  bone2Name=skirt_A_l_3
  swing1LimitDegrees=45
  swing1Motion=Free
  swing2LimitDegrees=45
  swing2Motion=Free
  twistLimitDegrees=45
  twistMotion=Free
 }
 {
  bone1Name=skirt_D_l_2
  bone2Name=skirt_C_l_2
  swing1LimitDegrees=45
  swing1Motion=Free
  swing2LimitDegrees=45
  swing2Motion=Free
  twistLimitDegrees=45
  twistMotion=Free
 }
 {
  bone1Name=skirt_D_l_2
  bone2Name=skirt_back_2
  swing1LimitDegrees=45
  swing1Motion=Free
  swing2LimitDegrees=45
  swing2Motion=Free
  twistLimitDegrees=5
  twistMotion=Free
 }
 {
  bone1Name=skirt_B_l_3
  bone2Name=skirt_C_l_3
  swing1LimitDegrees=45
  swing1Motion=Free
  swing2LimitDegrees=45
  swing2Motion=Free
  twistLimitDegrees=45
  twistMotion=Free
 }
]
