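# Post-process style AnimBlueprint for the F_MED_MincePounce body cosmetic, adding skirt motion and matching chin and jaw bones to the head mesh.
# The copied parent pose feeds a skirt Control Rig driven by movement and front-end state.
# Six facial bones are then replaced with head-mesh transforms cached each update.
# Pelvis-space RigidBody simulates the skirt with computed gravity override, before an additive back-skirt pitch derived from body-mesh skirt bones.
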AnimBlueprint /BRCosmetics/Characters/Player/Female/Medium/Bodies/F_MED_MincePounce/Meshes/F_MED_MincePounce_AnimBP
target_skeleton=&/F_MED_MincePounce_Skeleton
graphs [
 AnimGraph {nodes [
   "Transform (Modify) Bone - Bone: MSTR_skirt_C_r"!(ComponentPose, bAlphaBoolEnabled:bool=Get bIsInFrontEnd(), Translation="0,0,1", Rotation="0,0,0", Scale="1,1,1") -> (Pose) {
    alphaInputType=Bool
    rotationSpace=BCS_BoneSpace
    translationMode=BMM_Additive
    translationSpace=BCS_BoneSpace
    boneToModify {boneName=MSTR_skirt_C_r}
    translation {z=1}
   }
   n1 = Control Rig!(
    Source=Copy Pose From Mesh() {bUseAttachedParent=true}
    Alpha
    is staying:double=Get is staying() -> (is staying:double)
    is skydive:double=Get is skydive() -> (is skydive:double)
    in_air:double=Get in air() -> (in air:double)
    is frontend:bool=Get bIsInFrontEnd()
   ) -> (Pose) {controlRigAssetReference {blueprintRigClass=&/F_MED_MincePounce_CtrlRig_skirt.F_MED_MincePounce_CtrlRig_skirt_C}}
   n2 = Get!() -> (FACIAL_C_UnderChin_t_Location:Vector, FACIAL_C_UnderChin_t_Rotation:Rotator, FACIAL_C_UnderChin_t_Scale:Vector)
   n3 = "Transform (Modify) Bone - Bone: FACIAL_C_UnderChin"(ComponentPose=Local To Component(LocalPose=@n1), Alpha=1, Translation=@n2, Rotation=@n2.FACIAL_C_UnderChin_t_Rotation, Scale="1,1,1") {
    rotationMode=BMM_Replace
    rotationSpace=BCS_ParentBoneSpace
    translationMode=BMM_Replace
    translationSpace=BCS_ParentBoneSpace
    boneToModify {boneName=FACIAL_C_UnderChin}
   }
   n4 = Get!() -> (FACIAL_R_UnderChin_t_Location:Vector, FACIAL_R_UnderChin_t_Rotation:Rotator, FACIAL_R_UnderChin_t_Scale:Vector)
   n5 = "Transform (Modify) Bone - Bone: FACIAL_R_UnderChin"(ComponentPose=@n3, Alpha=1, Translation=@n4, Rotation=@n4.FACIAL_R_UnderChin_t_Rotation, Scale="1,1,1") {
    rotationMode=BMM_Replace
    rotationSpace=BCS_ParentBoneSpace
    translationMode=BMM_Replace
    translationSpace=BCS_ParentBoneSpace
    boneToModify {boneName=FACIAL_R_UnderChin}
   }
   n6 = Get!() -> (FACIAL_L_UnderChin_t_Location:Vector, FACIAL_L_UnderChin_t_Rotation:Rotator, FACIAL_L_UnderChin_t_Scale:Vector)
   n7 = "Transform (Modify) Bone - Bone: FACIAL_L_UnderChin"(ComponentPose=@n5, Alpha=1, Translation=@n6, Rotation=@n6.FACIAL_L_UnderChin_t_Rotation, Scale="1,1,1") {
    rotationMode=BMM_Replace
    rotationSpace=BCS_ParentBoneSpace
    translationMode=BMM_Replace
    translationSpace=BCS_ParentBoneSpace
    boneToModify {boneName=FACIAL_L_UnderChin}
   }
   n8 = Get!() -> (FACIAL_C_Jaw_t_Location:Vector, FACIAL_C_Jaw_t_Rotation:Rotator, FACIAL_C_Jaw_t_Scale:Vector)
   n9 = "Transform (Modify) Bone - Bone: FACIAL_C_Jaw"(ComponentPose=@n7, Alpha=1, Translation=@n8, Rotation=@n8.FACIAL_C_Jaw_t_Rotation, Scale="1,1,1") {
    rotationMode=BMM_Replace
    rotationSpace=BCS_ParentBoneSpace
    translationMode=BMM_Replace
    translationSpace=BCS_ParentBoneSpace
    boneToModify {boneName=FACIAL_C_Jaw}
   }
   n10 = Get!() -> (FACIAL_C_Chin_t_Location:Vector, FACIAL_C_Chin_t_Rotation:Rotator, FACIAL_C_Chin_t_Scale:Vector)
   n11 = "Transform (Modify) Bone - Bone: FACIAL_C_Chin"(ComponentPose=@n9, Alpha=1, Translation=@n10, Rotation=@n10.FACIAL_C_Chin_t_Rotation, Scale="1,1,1") {
    rotationMode=BMM_Replace
    rotationSpace=BCS_ParentBoneSpace
    translationMode=BMM_Replace
    translationSpace=BCS_ParentBoneSpace
    boneToModify {boneName=FACIAL_C_Chin}
   }
   n12 = Get!() -> (FACIAL_C_Jawline_t_Location:Vector, FACIAL_C_Jawline_t_Rotation:Rotator, FACIAL_C_Jawline_t_Scale:Vector)
   n13 = "Transform (Modify) Bone - Bone: FACIAL_C_Jawline"(ComponentPose=@n11, Alpha=1, Translation=@n12, Rotation=@n12.FACIAL_C_Jawline_t_Rotation, Scale="1,1,1") {
    rotationMode=BMM_Replace
    rotationSpace=BCS_ParentBoneSpace
    translationMode=BMM_Replace
    translationSpace=BCS_ParentBoneSpace
    boneToModify {boneName=FACIAL_C_Jawline}
   }
   n14 = RigidBody!(ComponentPose=@n13, Alpha, OverrideWorldGravity:Vector=Get Gravity Override(), ExternalForce) -> (Pose) {
    bOverrideWorldGravity=true
    overridePhysicsAsset=&/F_MED_MincePounce_Physics_Skirt
    simulationSpace=BaseBoneSpace
    baseBoneRef {boneName=pelvis}
    componentAppliedLinearAccClamp=<0 0 0>
    simSpaceSettings {maxAngularAcceleration=5, maxAngularVelocity=3, maxLinearAcceleration=150, maxLinearVelocity=150, worldAlpha=1}
   }
   n15 = Get mid_back_skirt_add() -> (mid_back_skirt_add:double)
   Output Pose(Result=Component To Local(ComponentPose="Transform (Modify) Bone - Bone: MSTR_skirt_back"!(ComponentPose=@n14, Alpha=1, Translation="0,0,0", Rotation_Roll:float=0, Rotation_Pitch:float=@n15, Rotation_Yaw:float=0, Scale="1,1,1") -> (Pose) {rotationMode=BMM_Additive, rotationSpace=BCS_BoneSpace, boneToModify {boneName=MSTR_skirt_back}})) {layerGroup=None}
  ]}
 EventGraph {nodes [
   Event BlueprintUpdateAnimation()
   > n1 = Sequence()
   > n2 = Gravity Override State Function DEPRECATED(
    CharacterPartAnimInstance=Self-Reference()
    GameTimeInSeconds=GetGameTimeInSeconds()
    GravityOverrideParams=Get Gravity Override Params()
    RBAN Params=Get RBAN Params()
   )
   > Set Gravity Override(Gravity Override=@n2)
   > n3 = Branch(Condition=Get bIsSkyDiving())
   > Set is staying(is staying:double=0) -> (Output_Get:double)
   > n4 = Set is skydive(is skydive:double=1) -> (Output_Get:double)
   n5 = Branch(execute=[@n4.then, @n9.then], Condition=Get bIsInAir() -> (bIsInAir:bool))
   > n6 = Set in air(in air:double=1) -> (Output_Get:double)
   Is Valid(exec=[@n7.then, @n6.then], InputObject=Get head_skeletal_mesh() -> (head_skeletal_mesh:Skeletal Mesh Component Budgeted Object Reference))
   > Set FACIAL_C_UnderChin_t(FACIAL_C_UnderChin_t:Transform=GetBoneTransform!(
     self:Skinned Mesh Component Object Reference=Get head_skeletal_mesh()
     InBoneName:Name=FACIAL_C_UnderChin
     TransformSpace:ERelativeTransformSpace Enum=RTS_ParentBoneSpace
    ) -> (ReturnValue:Transform)) -> (Output_Get:Transform)
   > Set FACIAL_L_UnderChin_t(FACIAL_L_UnderChin_t:Transform=GetBoneTransform(self=Get head_skeletal_mesh(), InBoneName=FACIAL_L_UnderChin, TransformSpace=RTS_ParentBoneSpace)) -> (Output_Get:Transform)
   > Set FACIAL_R_UnderChin_t(FACIAL_R_UnderChin_t:Transform=GetBoneTransform(self=Get head_skeletal_mesh(), InBoneName=FACIAL_R_UnderChin, TransformSpace=RTS_ParentBoneSpace)) -> (Output_Get:Transform)
   > Set FACIAL_C_Jaw_t(FACIAL_C_Jaw_t:Transform=GetBoneTransform(self=Get head_skeletal_mesh(), InBoneName=FACIAL_C_Jaw, TransformSpace=RTS_ParentBoneSpace)) -> (Output_Get:Transform)
   > Set FACIAL_C_Chin_t(FACIAL_C_Chin_t:Transform=GetBoneTransform(self=Get head_skeletal_mesh(), InBoneName=FACIAL_C_Chin, TransformSpace=RTS_ParentBoneSpace)) -> (Output_Get:Transform)
   > Set FACIAL_C_Jawline_t(FACIAL_C_Jawline_t:Transform=GetBoneTransform(self=Get head_skeletal_mesh(), InBoneName=FACIAL_C_Jawline, TransformSpace=RTS_ParentBoneSpace)) -> (Output_Get:Transform)
   n7 = Set in air(execute=@n5.else, in air=0)
   n8 = Get PawnVelocity()
   Set is staying(execute=@n3.else, is staying=MapRangeClamped(Value=Vector Length(A=@n8), InRangeA=0, InRangeB=100, OutRangeA=1, OutRangeB=0))
   > n9 = Set is skydive(is skydive=0)
   Is Valid(exec=@n1.then_1, InputObject=Get body skeletal mesh() -> (body skeletal mesh:Skeletal Mesh Component Object Reference))
   n10 = GetBoneTransform!(self=Get body skeletal mesh(), InBoneName=skirt_back_1, TransformSpace=RTS_ParentBoneSpace) -> (
    ReturnValue_Location:Vector
    ReturnValue_Rotation_Roll:float
    ReturnValue_Rotation_Pitch:float
    ReturnValue_Rotation_Yaw:float
    ReturnValue_Scale:Vector
   )
   n11 = "Max (Float)"(
    A=GetBoneTransform(self=Get body skeletal mesh(), InBoneName=skirt_D_l_1, TransformSpace=RTS_ParentBoneSpace).ReturnValue_Rotation_Pitch
    B=GetBoneTransform(self=Get body skeletal mesh(), InBoneName=skirt_D_r_1, TransformSpace=RTS_ParentBoneSpace).ReturnValue_Rotation_Pitch
   )
   > Set mid_back_skirt_add(mid_back_skirt_add:double=MapRangeClamped(Value=float - float(A=@n10.ReturnValue_Rotation_Pitch, B=@n11), InRangeA=-10, InRangeB=3, OutRangeA=18, OutRangeB=0)) -> (Output_Get:double)
   n12 = Event OnCharacterCustomizationLoaded()
   > Set body skeletal mesh(body skeletal mesh:Skeletal Mesh Component Object Reference=Get Array Item with EFort Custom Part Type(Array=@n12.CustomMeshParts)) -> (Output_Get:Skeletal Mesh Component Object Reference)
   > Set head_skeletal_mesh(head_skeletal_mesh:Skeletal Mesh Component Budgeted Object Reference=Get Array Item with EFort Custom Part Type(Array=@n12.CustomMeshParts, PartType=Head)) -> (Output_Get:Skeletal Mesh Component Budgeted Object Reference)
  ]}
]
variables [
 body skeletal mesh {type=SkeletalMeshComponent}
 FACIAL_C_Chin_t {type=Transform, default {location=<0 0 0>, rotation=<rot 0 0 0>, scale=<1 1 1>}}
 FACIAL_C_Jaw_t {type=Transform, default {location=<0 0 0>, rotation=<rot 0 0 0>, scale=<1 1 1>}}
 FACIAL_C_Jawline_t {type=Transform, default {location=<0 0 0>, rotation=<rot 0 0 0>, scale=<1 1 1>}}
 FACIAL_C_UnderChin_t {type=Transform, default {location=<0 0 0>, rotation=<rot 0 0 0>, scale=<1 1 1>}}
 FACIAL_L_UnderChin_t {type=Transform, default {location=<0 0 0>, rotation=<rot 0 0 0>, scale=<1 1 1>}}
 FACIAL_R_UnderChin_t {type=Transform, default {location=<0 0 0>, rotation=<rot 0 0 0>, scale=<1 1 1>}}
 Gravity Override {type=Vector}
 Gravity Override Params {type=GravityOverrideParamsStruct}
 head_skeletal_mesh {type=SkeletalMeshComponentBudgeted}
 in air {type=number}
 is skydive {type=number}
 is staying {type=number}
 mid_back_skirt_add {type=number}
 RBAN Params {
  type=RBANParamsStruct
  default {
   dBNO {simSpaceSettings {
     maxAngularAcceleration=1000
     maxAngularVelocity=3
     maxLinearAcceleration=300
     maxLinearVelocity=700
     velocityScaleZ=0
     externalLinearDragV=<0 0.5 0>
    }}
   dBNO_Carried {simSpaceSettings {
     maxAngularAcceleration=1000
     maxAngularVelocity=3
     maxLinearAcceleration=300
     maxLinearVelocity=700
     velocityScaleZ=0
     externalLinearDragV=<0 0.5 0>
    }}
   emoteOrMelee {simSpaceSettings {
     maxAngularAcceleration=1000
     maxAngularVelocity=3
     maxLinearAcceleration=300
     maxLinearVelocity=700
     velocityScaleZ=0
     externalLinearDragV=<0 0.5 0>
    }}
   front_End {simSpaceSettings {
     maxAngularAcceleration=1000
     maxAngularVelocity=3
     maxLinearAcceleration=300
     maxLinearVelocity=700
     velocityScaleZ=0
     externalLinearDragV=<0 0.5 0>
    }}
   onGround_Crouch_Moving {simSpaceSettings {
     maxAngularAcceleration=1000
     maxAngularVelocity=3
     maxLinearAcceleration=300
     maxLinearVelocity=700
     velocityScaleZ=0
     externalLinearDragV=<0 0.5 0>
    }}
   onGround_Falling {simSpaceSettings {
     maxAngularAcceleration=1000
     maxAngularVelocity=3
     maxLinearAcceleration=300
     maxLinearVelocity=700
     velocityScaleZ=0
     externalLinearDragV=<0 0.5 0>
    }}
   onGround_Moving {simSpaceSettings {
     maxAngularAcceleration=1000
     maxAngularVelocity=3
     maxLinearAcceleration=300
     maxLinearVelocity=700
     velocityScaleZ=0
     externalLinearDragV=<0 0.5 0>
    }}
   onGround_Standing {simSpaceSettings {
     maxAngularAcceleration=1000
     maxAngularVelocity=3
     maxLinearAcceleration=300
     maxLinearVelocity=700
     velocityScaleZ=0
     externalLinearDragV=<0 0.5 0>
    }}
   rail_Grinding {simSpaceSettings {
     maxAngularAcceleration=1000
     maxAngularVelocity=3
     maxLinearAcceleration=300
     maxLinearVelocity=700
     velocityScaleZ=0
     worldAlpha=0.1
     externalLinearDragV=<0 0.5 0>
    }}
   riding_Moving {simSpaceSettings {
     maxAngularAcceleration=1000
     maxAngularVelocity=3
     maxLinearAcceleration=300
     maxLinearVelocity=700
     velocityScaleZ=0
     externalLinearDragV=<0 0.5 0>
    }}
   riding_Moving_Targeting {simSpaceSettings {
     maxAngularAcceleration=1000
     maxAngularVelocity=3
     maxLinearAcceleration=300
     maxLinearVelocity=700
     velocityScaleZ=0
     externalLinearDragV=<0 0.5 0>
    }}
   riding_Standing {simSpaceSettings {
     maxAngularAcceleration=1000
     maxAngularVelocity=3
     maxLinearAcceleration=300
     maxLinearVelocity=700
     velocityScaleZ=0
     externalLinearDragV=<0 0.5 0>
    }}
   skydive_DivingDown {simSpaceSettings {
     maxAngularAcceleration=1000
     maxAngularVelocity=3
     maxLinearAcceleration=300
     maxLinearVelocity=700
     velocityScaleZ=0
     externalLinearDragV=<0 0.5 0>
    }}
   skydive_FloatingDown {
    componentLinearAccScale=<0 0.5 0>
    componentLinearVelScale=<0 0.5 0>
    simSpaceSettings {
     maxAngularAcceleration=1000
     maxAngularVelocity=3
     maxLinearAcceleration=300
     maxLinearVelocity=700
     velocityScaleZ=0
     externalLinearDragV=<0 0.5 0>
    }
   }
   skydive_GoingUp {simSpaceSettings {
     maxAngularAcceleration=1000
     maxAngularVelocity=3
     maxLinearAcceleration=300
     maxLinearVelocity=700
     velocityScaleZ=0
     externalLinearDragV=<0 0.5 0>
    }}
   skydive_Parachuting {
    componentLinearAccScale=<0 0.1 0.3>
    componentLinearVelScale=<0 0.1 0.3>
    simSpaceSettings {
     maxAngularAcceleration=1000
     maxAngularVelocity=3
     maxLinearAcceleration=300
     maxLinearVelocity=700
     velocityScaleZ=0
     externalLinearDragV=<0 0.5 0>
    }
   }
   skydive_Surfing {simSpaceSettings {
     maxAngularAcceleration=1000
     maxAngularVelocity=3
     maxLinearAcceleration=300
     maxLinearVelocity=700
     velocityScaleZ=0
     externalLinearDragV=<0 0.5 0>
    }}
   swimming {simSpaceSettings {
     maxAngularAcceleration=1000
     maxAngularVelocity=3
     maxLinearAcceleration=300
     maxLinearVelocity=700
     velocityScaleZ=0
     externalLinearDragV=<0 0.5 0>
    }}
   tactical_Sprint {simSpaceSettings {
     maxAngularAcceleration=1000
     maxAngularVelocity=3
     maxLinearAcceleration=300
     maxLinearVelocity=700
     velocityScaleZ=0
     externalLinearDragV=<0 0.5 0>
    }}
   ziplining {simSpaceSettings {
     maxAngularAcceleration=1000
     maxAngularVelocity=3
     maxLinearAcceleration=300
     maxLinearVelocity=700
     velocityScaleZ=0
     externalLinearDragV=<0 0.5 0>
    }}
  }
 }
 Return Value Rotation {type=Rotator}
]
