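# Physics asset for the bangs on the MincePounce female medium character's face accessory mesh, defining bodies and joints for hair physics.
# Left and right chains of three capsule bodies each link through constraints to a small MSTR root capsule.
# Four capsules on the head bone approximate the skull as a collision shape.
# Joints have tight twist limits that loosen toward the tips, and the lower left joints allow wider swing than the right.
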PhysicsAsset /BRCosmetics/Characters/Player/Female/Medium/Bodies/F_MED_MincePounce/Meshes/Parts/F_MED_MincePounce_FaceAcc_Physics_Bangs
bodies {
 MSTR_bang_l_1 [{shapeType=Capsule, length=0.505, radius=0.505, shapeName=MSTR_bang_l_1_capsule, rotation=<rot 0 -90 0>}]
 MSTR_bang_r_1 [{shapeType=Capsule, length=0.505, radius=0.505, shapeName=MSTR_bang_r_1_capsule, rotation=<rot 0 -90 0>}]
 bang_l_1 [{shapeType=Capsule, length=2.8441, radius=1.4043, shapeName=bang_l_1_capsule, center=<2.6584 0 0>, rotation=<rot 84.4001 0 0>}]
 bang_l_2 [{
   shapeType=Capsule
   length=1.1964
   radius=2.1676
   shapeName=bang_l_2_capsule
   center=<2.6188 0.0444 1.3875>
   rotation=<rot 86.3946 -164.5004 180>
  }]
 bang_l_3 [{shapeType=Capsule, length=3.3149, radius=1.937, shapeName=bang_l_3_capsule, center=<3.4447 0.7954 0.7716>, rotation=<rot 77.1675 0.0001 0>}]
 bang_r_1 [{shapeType=Capsule, length=2.7196, radius=1.5288, shapeName=bang_r_1_capsule, center=<2.8335 0.1776 -0.0951>, rotation=<rot 82.8434 0 0>}]
 bang_r_2 [{
   shapeType=Capsule
   length=2.5323
   radius=1.7161
   shapeName=bang_r_2_capsule
   center=<2.6312 -0.1744 0.8943>
   rotation=<rot 82.6409 171.1223 -179.9999>
  }]
 bang_r_3 [{
   shapeType=Capsule
   length=3.2397
   radius=2.0122
   shapeName=bang_r_3_capsule
   center=<3.5791 -1.2622 0.5665>
   rotation=<rot 73.4747 0.1371 0.1438>
  }]
 head [
  {
   shapeType=Capsule
   length=7.1949
   radius=7.9408
   shapeName=head_capsule
   center=<3.9636 -3.1139 0.1479>
   rotation=<rot -0.3019 -92.2671 -88.9053>
  }
  {shapeType=Capsule, length=4.9987, radius=5.6599, shapeName=head_capsule, center=<4.4028 4.9971 0>, rotation=<rot 90 0 0>}
  {shapeType=Capsule, length=6.8362, radius=2.3601, shapeName=head_capsule, center=<4.8054 0.2092 6.9621>, rotation=<rot 90 7.4898 0>}
  {shapeType=Capsule, length=4.6712, radius=2.2652, shapeName=head_capsule, center=<4.2858 0.4952 -6.7651>, rotation=<rot 90 0 0>}
 ]
}
constraints [
 {bone1Name=bang_r_3, bone2Name=bang_r_2, swing1LimitDegrees=37, swing2LimitDegrees=37, twistLimitDegrees=5}
 {bone1Name=bang_r_2, bone2Name=bang_r_1, swing1LimitDegrees=37, swing2LimitDegrees=37, twistLimitDegrees=2}
 {bone1Name=bang_r_1, bone2Name=MSTR_bang_r_1, swing1LimitDegrees=37, swing2LimitDegrees=37, twistLimitDegrees=1}
 {bone1Name=bang_l_3, bone2Name=bang_l_2, swing1LimitDegrees=50, swing2LimitDegrees=50, twistLimitDegrees=5}
 {bone1Name=bang_l_2, bone2Name=bang_l_1, swing1LimitDegrees=50, swing2LimitDegrees=50, twistLimitDegrees=2}
 {bone1Name=bang_l_1, bone2Name=MSTR_bang_l_1, swing1LimitDegrees=37, swing2LimitDegrees=37, twistLimitDegrees=1}
]
